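; Component for reactive props that relays Barn Towel Armory lifecycle events to its own event dispatchers.
; Handlers for spawn, wing deployment, thruster activation, landing, opening, open and player entry rebroadcast each stage, passing the selected variant or pawn where relevant.
; The opened broadcast waits on a timer when DelayedPropOpenEvent is above zero, otherwise firing immediately.
; Debug strings print only without authority; the immediate open path prints none.
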(event Custom|OnArmoryEntered
  (params (FortPawn :type Fort%20Pawn%20Object%20Reference))
  (Default|CallOnArmoryPlayerEntered FortPawn)
  (Utilities|SwitchHasAuthority
    (:Remote
      (Development|PrintString "Armory Entered"))))

(event Custom|OnArmorySpawnedReceived
  (params (SpawnedVariant :type Enum_BarnTowelArmory_SelectedVariant%20Enum :default "NewEnumerator0"))
  (Default|CallOnArmorySpawned SpawnedVariant)
  (Utilities|SwitchHasAuthority
    (:Remote
      ; Initial Spawn
      (Development|PrintString "Armory Spawn Recieved"))))

(event Custom|OnArmoryWingsDeployedReceived
  (Default|CallOnArmoryWingsDeployed)
  (Utilities|SwitchHasAuthority
    (:Remote
      (Development|PrintString "Armory Wings Delployed"))))

(event Custom|OnArmoryThrustersActivatedReceived
  (Default|CallOnArmoryThrustersActivated)
  (Utilities|SwitchHasAuthority
    (:Remote
      ; Thrusters & Top Arms unfold
      (Development|PrintString "Armory Thrusters Activate"))))

(event Custom|OnArmoryLandedReceived
  (params (SpawnedVariant :type Enum_BarnTowelArmory_SelectedVariant%20Enum :default "NewEnumerator0"))
  (Default|CallOnArmoryLanded SpawnedVariant)
  (Utilities|SwitchHasAuthority
    (:Remote
      ; Fully Landed
      (Development|PrintString "Armory Landed Recieved"))))

(event Custom|OnArmoryOpeningReceived
  (Default|CallOnArmoryOpening)
  (Utilities|SwitchHasAuthority
    (:Remote
      (Development|PrintString "Armory Opening"))))

(event Custom|OnArmoryOpenReceived
  (bind _delayedpropopenevent (Variables|Default|GetDelayedPropOpenEvent))
  (if (> _delayedpropopenevent 0)
    (Utilities|Time|SetTimerbyEvent (EventDispatchers|CreateEvent :!event Custom|DelayedOpenEvent) _delayedpropopenevent) ; delay by the timer
    (else
      (Default|CallOnArmoryOpened) ; no delay
    )))

(event Custom|DelayedOpenEvent
  (Default|CallOnArmoryOpened)
  (Utilities|SwitchHasAuthority
    (:Remote
      (Development|PrintString "Armory Delayed Open"))))
